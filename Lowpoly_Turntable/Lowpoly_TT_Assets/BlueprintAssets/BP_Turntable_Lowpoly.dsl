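; Turntable controller Blueprint for presenting and checking low-poly models during capture.
; Construction gathers user, date and engine build, and applies display materials carrying each mesh's normal, AO and wireframe textures.
; BeginPlay disables AA, motion blur and Lumen GI, then feeds a HUD object names and asset-tag vertex counts; Tick updates its capture frame number.
; Functions switch grey, checker or wireframe materials and toggle UV density; an event steps through UV sheets.
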(event EventBeginPlay
  (Utilities|FlowControl|Sequence
    (:then_0
      (Development|ExecuteConsoleCommand "showflag.antialiasing 0")
      (Development|ExecuteConsoleCommand "showflag.motionblur 0")
      (Development|ExecuteConsoleCommand "r.SSR.Temporal 0")
      ;[1 1 1] Adding to remove ghosting from the previous frames
      (Development|ExecuteConsoleCommand "ShowFlag.LumenGlobalIllumination 0"))
    (:then_1
      ;[1 1 1] setup HUD
      (bind _returnvalue (UserInterface|CreateWidget "/Lowpoly_Turntable/Lowpoly_TT_Assets/BlueprintAssets/WBP_Turntable_UI_Lowpoly.WBP_Turntable_UI_Lowpoly_C"))
      (Variables|Default|SetUI _returnvalue)
      (UserInterface|Viewport|AddtoViewport _returnvalue 0)
      (Default|SetupObjectList)
      (Default|SetText _returnvalue (Variables|Default|GetuserName) (Variables|Default|Getdate) (Variables|Default|Getbuild) (Variables|Default|GetobjectList) "" "" (Variables|Default|Getversion)))))

(event Collision|EventActorBeginOverlap
  (params (OtherActor :type Actor%20Object%20Reference)))

(event EventTick
  (params (DeltaSeconds :type Float%20%28single-precision%29))
  (Default|SetFrameNumber (Variables|Default|GetUI) (Cinematics|Capture|GetCaptureFrameNumber)))

(event Custom|Show%20UV%20with%20Delay)

(event Custom|Set%20Current%20UV%20Event
  (bind _turntable_objects (Variables|Default|GetTurntable_Objects))
  (bind _rootcomponent (Variables|Transformation|GetRootComponent _turntable_objects))
  (bind _turntable_sets (Variables|Default|GetTurntable_Sets))
  (bind _rootcomponent_2 (Variables|Transformation|GetRootComponent _turntable_sets))
  (bind _uvnum (Variables|Default|GetUVnum))
  (bind _wireframetexture (Variables|Default|GetWireframeTexture))
  ;[1 1 1] UV Sheet Event
  (bind _returnvalue (Actor|GetActorOfClass "/Lowpoly_Turntable/Lowpoly_TT_Assets/BlueprintAssets/BP_SimpleTOD.BP_SimpleTOD_C"))
  (Rendering|SetVisibility (Variables|Transformation|GetRootComponent _returnvalue) false true)
  (Rendering|SetVisibility _rootcomponent_2 false true)
  (Rendering|SetVisibility _rootcomponent false true)
  (Rendering|SetVisibility (Variables|Default|GetUVSheet) true false)
  (if (> _uvnum (- (Utilities|Array|Length _wireframetexture) 1)) :!label _label
    (Variables|Default|SetUVnum 0)
    (Rendering|SetVisibility (Components|GetChildrenComponents _rootcomponent false) true false)
    (Rendering|SetVisibility (Variables|Default|GetUVSheet) false false)
    (Rendering|SetVisibility _rootcomponent_2 true false)
    (Rendering|SetVisibility (Variables|Transformation|GetRootComponent _returnvalue) true false)
    (else
      (bind _returnvalue_2 (Rendering|Material|CreateDynamicMaterialInstance (Variables|Default|GetUVSheet) 0 "/Lowpoly_Turntable/Lowpoly_TT_Assets/BlueprintAssets/MAT_Turntable_ModelChecker_Wireframe.MAT_Turntable_ModelChecker_Wireframe" "None" :!class /Script/Engine.PrimitiveComponent))
      (Rendering|Material|SetTextureParameterValue _returnvalue_2 "BaseColor" (Utilities|Array|Get(acopy) _wireframetexture _uvnum))
      (Utilities|FlowControl|Delay 0.75)
      (Variables|Default|SetUVnum (+ (Variables|Default|GetUVnum) 1))
      (goto _label))))

(fn checkVertexCount
  (for _array_element (Utilities|Array|MakeArray (Actor|GetComponentbyClass "/Script/Engine.SkeletalMeshComponent"))
    (bind _asstatic_mesh_component (Utilities|Casting|CastToStaticMeshComponent _array_element))
    (bind _returnvalue (Default|VertexCount (Variables|/Script/Engine.StaticMeshComponent|GetStaticMesh _asstatic_mesh_component)))
    (Development|PrintString (Utilities|String|Append "skeletalMesh " (Utilities|GetDisplayName (Variables|/Script/Engine.StaticMeshComponent|GetStaticMesh _asstatic_mesh_component)) " " (Utilities|String|ToString(Integer) _returnvalue)))))

(fn MaterialSwap-Checker
  (Variables|Default|SetDisplayMaterial "NewEnumerator1")
  (Default|MeshSetup-New))

(fn MaterialSwap-LIghtGrey
  (Variables|Default|SetDisplayMaterial "NewEnumerator0")
  (Default|MeshSetup-New))

(fn MaterialSwap-Wireframe
  (Variables|Default|SetDisplayMaterial "NewEnumerator2")
  (Default|MeshSetup-New))

(fn MaterialSwitch
  (params (Mesh%20 :type Primitive%20Component%20Object%20Reference) (Material :type Material%20Instance%20Dynamic%20Object%20Reference) (element%20number :type Integer))
  (bind _material (Variables|Default|GetMaterial))
  (switch Utilities|FlowControl|Switch|SwitchonDisplayMaterial (Variables|Default|GetDisplayMaterial)
    (:NewEnumerator0
      (Variables|Default|SetMaterial "/Lowpoly_Turntable/Lowpoly_TT_Assets/BlueprintAssets/MAT_Turntable_ModelChecker_Grey.MAT_Turntable_ModelChecker_Grey"))
    (:NewEnumerator1
      (Variables|Default|SetMaterial "/Lowpoly_Turntable/Lowpoly_TT_Assets/BlueprintAssets/MAT_Turntable_ModelChecker_Checker.MAT_Turntable_ModelChecker_Checker"))
    (:NewEnumerator2
      (Variables|Default|SetMaterial "/Lowpoly_Turntable/Lowpoly_TT_Assets/BlueprintAssets/MAT_Turntable_ModelChecker_BakedWireFrame.MAT_Turntable_ModelChecker_BakedWireFrame")))
  (Rendering|Material|SetMaterial Mesh%20 element%20number _material)
  (bind _returnvalue (Rendering|Material|CreateDynamicMaterialInstance Mesh%20 element%20number _material "None" :!class /Script/Engine.PrimitiveComponent))
  (Rendering|Material|SetTextureParameterValue _returnvalue "Normals" (Variables|Default|GetNormals))
  (Rendering|Material|SetTextureParameterValue _returnvalue "AO" (Variables|Default|GetAO))
  (Rendering|Material|SetTextureParameterValue _returnvalue "Wireframe" (Variables|Default|GetWireframe))
  (Rendering|Material|SetScalarParameterValue _returnvalue "Tiling" (Variables|Default|Gettiling) :!class /Script/Engine.MaterialInstanceDynamic))

(fn MeshSetup-New
  (Development|ExecuteConsoleCommand "r.ForceLOD 0")
  (Development|ExecuteConsoleCommand "r.SkeletalMeshLODBias -10")
  (Utilities|Array|Clear (Variables|Default|GetWireframeTexture))
  (for _array_element (Components|GetChildrenComponents (Variables|Transformation|GetRootComponent (Variables|Default|GetTurntable_Objects)) false)
    (if (Utilities|String|Contains (Utilities|GetDisplayName _array_element) "body")
      (Variables|Default|Settiling 5.0)
      (else
        (Variables|Default|Settiling 2.0)))
    (Development|PrintString (Utilities|GetDisplayName _array_element))
    (bind _asstatic_mesh_component (Utilities|Casting|CastToStaticMeshComponent _array_element))
    (for _index (range (Rendering|Material|GetNumMaterials _asstatic_mesh_component))
      (bind _returnvalue (Rendering|Material|CreateDynamicMaterialInstance _asstatic_mesh_component _index 0 "None" :!class /Script/Engine.PrimitiveComponent))
      (bind _returnvalue_2 (Rendering|Material|GetTextureParameterValue _returnvalue "Normals" :!class /Script/Engine.MaterialInstanceDynamic))
      (Variables|Default|SetNormals _returnvalue_2)
      (bind _returnvalue_3 (Rendering|Material|GetTextureParameterValue _returnvalue "AO" :!class /Script/Engine.MaterialInstanceDynamic))
      (Variables|Default|SetAO _returnvalue_3)
      (bind _returnvalue_4 (Rendering|Material|GetTextureParameterValue _returnvalue "Wireframe" :!class /Script/Engine.MaterialInstanceDynamic))
      (bind _output_get (Variables|Default|SetWireframe _returnvalue_4))
      (Utilities|Array|Add (Variables|Default|GetWireframeTexture) _output_get)
      (Default|MaterialSwitch _asstatic_mesh_component 0 _index))))

(fn SetCurrentUV)

(fn SetMovieEndFrame
  (if (Variables|Default|GetAdd%20frames%20for%20additional%20UV%20pages)
    _))

(fn setup)

(fn setupObjectList
  (bind _objectlist (Variables|Default|GetobjectList))
  (Variables|Default|SetobjectList "")
  (for _array_element (Components|GetChildrenComponents (Variables|Transformation|GetRootComponent (Variables|Default|GetTurntable_Objects)) false)
    (bind _asskeletal_mesh_component (Utilities|Casting|CastToSkeletalMeshComponent _array_element)
      (:then
        (bind _returnvalue (Default|VertexCountSkeletalMesh (Variables|/Script/Engine.SkinnedMeshComponent|GetSkeletalMesh _asskeletal_mesh_component)))
        (bind _array_element_2 (Utilities|Array|ForEachLoopwithBreak (Utilities|String|GetCharacterArrayfromString (Utilities|String|Reverse (.RightS (Utilities|String|Split (Utilities|GetDisplayName _asskeletal_mesh_component) "SkeletalMeshComponent0")))))
          (:LoopBody
            (if (Variables|Default|GetEndwithNumber)
              (if (Utilities|String|IsNumeric _array_element_2)
                (bind _output_get (Variables|Default|SetNumericCount (+ (Variables|Default|GetNumericCount) 1)))
                (else
                  (Variables|Default|SetEndwithNumber false)))))
          (:Completed
            (Variables|Default|SetobjectList (Utilities|String|Append _objectlist (Utilities|String|LeftChop (.RightS (Utilities|String|Split (Utilities|GetDisplayName _asskeletal_mesh_component) "SkeletalMeshComponent0")) _output_get) " (SK, Vert Count:" (Utilities|String|ToString(Integer) _returnvalue) ")
"))
            (Variables|Default|SetVertSum (+ (Variables|Default|GetVertSum) _returnvalue)))))
      (:CastFailed
        (bind _asstatic_mesh_component (Utilities|Casting|CastToStaticMeshComponent _array_element))
        (bind _returnvalue_2 (Default|VertexCount (Variables|/Script/Engine.StaticMeshComponent|GetStaticMesh _asstatic_mesh_component)))
        (bind _array_element_3 (Utilities|Array|ForEachLoopwithBreak (Utilities|String|GetCharacterArrayfromString (Utilities|String|Reverse (.RightS (Utilities|String|Split (Utilities|GetDisplayName _asstatic_mesh_component) "StaticMeshComponent0")))))
          (:LoopBody
            (if (Variables|Default|GetEndwithNumber)
              (if (Utilities|String|IsNumeric _array_element_3)
                (bind _output_get_2 (Variables|Default|SetNumericCount (+ (Variables|Default|GetNumericCount) 1)))
                (else
                  (Variables|Default|SetEndwithNumber false)))))
          (:Completed
            (Variables|Default|SetobjectList (Utilities|String|Append _objectlist (Utilities|String|LeftChop (.RightS (Utilities|String|Split (Utilities|GetDisplayName _asstatic_mesh_component) "StaticMeshComponent0")) _output_get_2) "(SM, Vert Count: " (Utilities|String|ToString(Integer) _returnvalue_2) ")
"))
            (Variables|Default|SetVertSum (+ (Variables|Default|GetVertSum) _returnvalue_2)))))))
  (Variables|Default|SetobjectList (Utilities|String|Append (Variables|Default|GetobjectList) (Utilities|String|Append "*Vert Count Sum: " (Utilities|String|ToString(Integer) (Variables|Default|GetVertSum))))))

(fn setupShadows
  (Utilities|FlowControl|Sequence
    (:then_0
      (Development|ExecuteConsoleCommand "r.Shadow.PerObjectDirectionalDepthBias 10"))
    (:then_1
      (Development|ExecuteConsoleCommand "r.Shadow.PerObjectDirectionalSlopeDepthBias 10"))
    (:then_2
      (for _array_element (Utilities|Array|MakeArray (Actor|GetComponentbyClass "/Script/Engine.SkeletalMeshComponent"))
        (bind _asprimitive_component (Utilities|Casting|CastToPrimitiveComponent _array_element))
        (Rendering|SetCastInsetShadow _asprimitive_component true)))))

(fn SetupWireFrame
  (if true
    (Development|ExecuteConsoleCommand "r.PostProcessAAQuality 0")
    (Development|ExecuteConsoleCommand "ShowFlag.MeshEdges 1")
    (else
      (Development|ExecuteConsoleCommand "r.PostProcessAAQuality 6")
      (Development|ExecuteConsoleCommand "ShowFlag.MeshEdges 0")))
  (bind _returnvalue (Actor|GetActorOfClass "/Lowpoly_Turntable/Lowpoly_TT_Assets/BlueprintAssets/BP_SimpleTOD.BP_SimpleTOD_C")) ; NEEDS SIMPLE TIME OF DAY
  (Rendering|SetVisibility (Variables|Transformation|GetRootComponent _returnvalue) false false)
  (for _array_element (Components|GetChildrenComponents (Variables|Transformation|GetRootComponent (Variables|Default|GetTurntable_Sets)) false)
    (Rendering|SetVisibility (Variables|Transformation|GetRootComponent) false false)))

(fn ConstructionScript
  (bind (_lefts _rights) (Utilities|String|Split (Utilities|String|Split (Development|GetEngineVersion) "+++") "-"))
  (Default|MeshSetup-New)
  ;[1 1 1] UserName
  (Variables|Default|SetuserName (Utilities|Platform|GetPlatformUserName))
  ;[1 1 1] Date
  (Variables|Default|Setdate (Utilities|String|ToString(Text) (Utilities|Text|AsDate(fromUTC) (Math|DateTime|UTCNow) "EDT")))
  ;[1 1 1] Build #
  (Variables|Default|Setbuild _rights)
  (Default|SetupShadows))

(fn UVDensityoff
  (Rendering|SetVisibility (Variables|Transformation|GetRootComponent (Variables|Default|GetTurntable_Sets)) true true)
  (Development|ExecuteConsoleCommand "Show MeshUVDensityAccuracy"))

(fn UVDensityon
  (Rendering|SetVisibility (Variables|Transformation|GetRootComponent (Variables|Default|GetTurntable_Sets)) false true)
  (Development|ExecuteConsoleCommand "Show MeshUVDensityAccuracy"))

(fn vertexCount
  (params (StaticlMesh :type Static%20Mesh%20Object%20Reference))
  (bind _returnvalue (Utilities|Name|MakeLiteralName "Vertices"))
  (bind _returnvalue_2 (AssetData|CreateAssetData StaticlMesh))
  (bind (_outtagvalue _returnvalue_3) (AssetData|GetTagValue _returnvalue_2 _returnvalue))
  (if _returnvalue_3
    (return (Utilities|String|StringToInteger _outtagvalue))
    (else
      (return -1))))

(fn vertexCount_SkeletalMesh
  (params (skeletalMEsh :type Skeletal%20Mesh%20Object%20Reference))
  (bind _returnvalue (Utilities|Name|MakeLiteralName "Vertices"))
  (bind _returnvalue_2 (AssetData|CreateAssetData skeletalMEsh))
  (bind (_outtagvalue _returnvalue_3) (AssetData|GetTagValue _returnvalue_2 _returnvalue))
  (if _returnvalue_3
    (return (Utilities|String|StringToInteger _outtagvalue))
    (else
      (return -1))))
