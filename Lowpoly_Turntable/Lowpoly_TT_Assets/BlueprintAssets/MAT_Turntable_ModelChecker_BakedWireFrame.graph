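# A Default Lit material for the low-poly turntable's model checker, named for displaying a baked wireframe on the model.
# The Wireframe texture parameter, blank by default, is inverted, raised to a power to sharpen its contrast, then scaled down into a dimmed mask.
# A Normals texture parameter takes a normal map and defaults to a flat placeholder, alongside two standalone constants.
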Material /Lowpoly_Turntable/Lowpoly_TT_Assets/BlueprintAssets/MAT_Turntable_ModelChecker_BakedWireFrame
graph {
 properties {pixelDepthOffsetMode=PDOM_Legacy, usedShadingModels=Default Lit}
 nodes [
  Constant() {r=1}
  Constant()
  n1 = TextureSampleParameter2D() {parameterName=Wireframe, texture=&/T_BlankAlpha}
  Multiply(A=Power(Base=OneMinus(None=@n1.RGB)) {constExponent=4}, B=Constant() {r=0.3})
  TextureSampleParameter2D() {parameterName=Normals, samplerType=SAMPLERTYPE_Normal, texture=/Paper2D/PlaceholderTextures/FlatNormalMap}
 ]
}
