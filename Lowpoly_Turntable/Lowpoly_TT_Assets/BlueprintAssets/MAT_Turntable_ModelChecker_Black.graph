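# Unlit material that, by its name, serves as the black material for the low-poly turntable model checker.
# Its graph holds a single scalar constant at a near-black value, with no textures, parameters or math.
# The unlit shading model keeps scene lighting from affecting its appearance.
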Material /Lowpoly_Turntable/Lowpoly_TT_Assets/BlueprintAssets/MAT_Turntable_ModelChecker_Black
graph {properties {pixelDepthOffsetMode=PDOM_Legacy, shadingModel=MSM_Unlit, usedShadingModels=Unlit}, nodes [Constant() {r=0.01}]}
