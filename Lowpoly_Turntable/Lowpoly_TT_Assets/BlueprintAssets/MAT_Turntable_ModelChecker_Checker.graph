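# Default Lit material that, judging by its name, shows a checker pattern for inspecting models in a low-poly turntable.
# Checker texture UVs scale by a Tiling parameter, and a slight blend toward white softens the pattern.
# The AO and "a" texture parameters are each offset by an Add, multiplied, saturated, then multiplied into the checker.
# NormalMap and Normals parameters sit unlinked, and no material output connections are shown.
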Material /Lowpoly_Turntable/Lowpoly_TT_Assets/BlueprintAssets/MAT_Turntable_ModelChecker_Checker
graph {
 properties {pixelDepthOffsetMode=PDOM_Legacy, usedShadingModels=Default Lit}
 nodes [
  Constant()
  n1 = Add(A=TextureSampleParameter2D() {parameterName=AO, texture=&/T_BlankAlpha}.RGB)
  n2 = Add(A=TextureSampleParameter2D() {parameterName=a, texture=&T_BlankAlpha}.RGB)
  n3 = Multiply(A=TextureCoordinate() {uTiling=2, vTiling=2}, B=ScalarParameter() {defaultValue=2, parameterName=Tiling})
  Multiply(
   A=Saturate(None=Multiply(A=@n1, B=@n2))
   B=LinearInterpolate(
    A=TextureSample(UVs=@n3) {texture=&/Checker}.RGB
    B=Constant() {r=1}
    Alpha=Constant() {r=0.07}
   )
  )
  TextureSampleParameter2D() {parameterName=NormalMap, samplerType=SAMPLERTYPE_Normal, texture=/Engine/EngineMaterials/FlatNormal}
  TextureSampleParameter2D() {parameterName=Normals, samplerType=SAMPLERTYPE_Normal, texture=/Paper2D/PlaceholderTextures/FlatNormalMap}
 ]
}
