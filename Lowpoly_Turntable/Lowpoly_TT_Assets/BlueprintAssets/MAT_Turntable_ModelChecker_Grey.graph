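# Default Lit turntable material, named as a grey model checker, with texture parameters for Wireframe, AO and Normals.
# The Wireframe sample passes through an add and saturate, then multiplies an AO term made from a power curve scaled by a small constant.
# Wireframe and AO default to a blank alpha texture, and Normals defaults to a flat normal map.
# The graph shows no connections to material outputs.
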Material /Lowpoly_Turntable/Lowpoly_TT_Assets/BlueprintAssets/MAT_Turntable_ModelChecker_Grey
graph {
 properties {pixelDepthOffsetMode=PDOM_Legacy, usedShadingModels=Default Lit}
 nodes [
  Constant() {r=0.5}
  Constant() {r=1}
  n1 = TextureSampleParameter2D() {parameterName=Wireframe, texture=&/T_BlankAlpha}
  n2 = TextureSampleParameter2D() {parameterName=AO, texture=&T_BlankAlpha}
  n3 = Constant() {r=0.3}
  Multiply(A=Saturate(None=Add(A=@n1.RGB)), B=Multiply(A=Constant() {r=0.05}, B=Power(Base=@n2.RGB, Exp=@n3)))
  TextureSampleParameter2D() {parameterName=Normals, samplerType=SAMPLERTYPE_Normal, texture=/Paper2D/PlaceholderTextures/FlatNormalMap}
 ]
}
