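# Unlit material in the Lowpoly Turntable assets, named for the model checker's wireframe view.
# Its node chain inverts the RGB of a BaseColor texture parameter, which defaults to a black texture, then raises it to a high constant power.
# This turns black to white and quickly darkens lighter values.
# The graph shows no link from that result to a material output and no wireframe setting.
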Material /Lowpoly_Turntable/Lowpoly_TT_Assets/BlueprintAssets/MAT_Turntable_ModelChecker_Wireframe
graph {
 properties {pixelDepthOffsetMode=PDOM_Legacy, shadingModel=MSM_Unlit, usedShadingModels=Unlit}
 nodes [Power(Base=OneMinus(None=TextureSampleParameter2D() {parameterName=BaseColor, texture=&/T_BlacksRGB}.RGB)) {constExponent=10}]
}
