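# Unlit, two-sided wireframe master material for the low-poly turntable that displays a mesh as its UV unwrap.
# UV coordinates become scaled points on the XZ plane, combined with negated world position and a fixed offset to compute vertex displacement into that layout.
# The UVSwitch scalar blends between the negated and original result, defaulting to the original.
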Material /Lowpoly_Turntable/Lowpoly_TT_Assets/BlueprintAssets/MAT_UVUnwrap_MASTER
graph {
 properties {pixelDepthOffsetMode=PDOM_Legacy, shadingModel=MSM_Unlit, twoSided=true, usedShadingModels=Unlit, wireframe=true}
 nodes [
  n1 = MaterialFunctionCall(Float2=TextureCoordinate()) {materialFunction=&BreakOutFloat2Components}
  n2 = AppendVector(A=@n1.R, B=Constant())
  n3 = WorldPosition()
  n4 = Constant3Vector() {constant=<rgba -50 0 50 0>}
  Add(A=Multiply(A=AppendVector(A=@n2, B=@n1.G)) {constB=100}, B=Multiply(A=Subtract(A=@n3.XYZ, B=@n4)) {constB=-1})
  n5 = Add(B=@n4)
  LinearInterpolate(A=Multiply(A=@n5) {constB=-1}, B=@n5, Alpha=ScalarParameter() {defaultValue=1, parameterName=UVSwitch})
 ]
}
