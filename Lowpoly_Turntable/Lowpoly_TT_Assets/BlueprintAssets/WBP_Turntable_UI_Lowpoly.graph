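# Movie Render Pipeline burn-in widget that overlays turntable render info: name, date, build, object list, pass, frame number and version.
# On Construct it takes the first BP_Turntable_Lowpoly found, registers itself as that actor's UI and copies its strings into the text blocks, leaving pass and frame blank.
# OnOutputFrameStarted writes the pipeline's root frame number each output frame.
# A setPass function updates the pass label but is not called within this widget.
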WidgetBlueprint /Lowpoly_Turntable/Lowpoly_TT_Assets/BlueprintAssets/WBP_Turntable_UI_Lowpoly
graphs [
 EventGraph {nodes [
   Event Construct()
   > n1 = GetAllActorsOfClass!(execute, ActorClass:Actor Class Reference=&/BP_Turntable_Lowpoly.BP_Turntable_Lowpoly_C) -> (then, OutActors:Array of BP Turntable Lowpoly Object References)
   n2 = "Get (a copy)"(Array:Array of BP Turntable Lowpoly Object References=@n1) -> (Output:BP Turntable Lowpoly Object Reference)
   > Set UI!(execute, UI:WBP Turntable UI Lowpoly Object Reference=Self-Reference(), self:BP Turntable Lowpoly Object Reference=@n2) -> (then, Output_Get:WBP Turntable UI Lowpoly Object Reference)
   > setText!(
    execute
    self
    Name:String=Get userName!(self:BP Turntable Lowpoly Object Reference=@n2) -> (userName:String)
    Date:String=Get date!(self:BP Turntable Lowpoly Object Reference=@n2) -> (date:String)
    build:String=Get build!(self:BP Turntable Lowpoly Object Reference=@n2) -> (build:String)
    objectList:String=Get objectList!(self:BP Turntable Lowpoly Object Reference=@n2) -> (objectList:String)
    Pass:String
    FrameNumber:String
    Version:String=Get version!(self:BP Turntable Lowpoly Object Reference=@n2) -> (version:String)
   ) -> (then)
   n3 = Event OnOutputFrameStarted!() -> (OutputDelegate, then, ForPipeline:Movie Pipeline Object Reference) {is=title K2Node_Event}
   > setFrameNumber!(
    execute
    self
    frameNumber:int=Break Frame Number!(FrameNumber:"Frame Number Structure (by ref)"=GetRootFrameNumber!(InMoviePipeline:Movie Pipeline Object Reference=@n3.ForPipeline) -> (ReturnValue:Frame Number Structure)) -> (Value:int) {is=title K2Node_BreakStruct}
   ) -> (then)
   setFrameNumber(frameNumber=FrameNumber to Integer!(InFrameNumber:"Frame Number Structure (by ref)"=GetCurrentShotFrameNumber!(InMoviePipeline:Movie Pipeline Object Reference) -> (ReturnValue:Frame Number Structure)) -> (ReturnValue:int=0))
   Event PreConstruct()
   Event Tick()
  ]}
 setFrameNumber {nodes [
   n1 = setFrameNumber!() -> (then, frameNumber:int) {is=title K2Node_FunctionEntry}
   > "SetText (Text)"(self=Get frameNumber() -> (frameNumber:Text Object Reference), InText="To Text (Integer)"(Value=@n1, MaximumIntegralDigits=800))
  ]}
 setPass {nodes [
   n1 = setPass!() -> (then, Pass:String) {is=title K2Node_FunctionEntry}
   > "SetText (Text)"(self=Get Pass() -> (Pass:Text Object Reference), InText="To Text (String)"(InString=@n1))
  ]}
 setText {nodes [
   n1 = setText!() -> (then, Name:String, Date:String, build:String, objectList:String, Pass:String, FrameNumber:String, Version:String) {is=title K2Node_FunctionEntry}
   > "SetText (Text)"(self=Get Name() -> (Name:Text Object Reference), InText="To Text (String)"(InString=@n1))
   > "SetText (Text)"(self=Get Date() -> (Date:Text Object Reference), InText="To Text (String)"(InString=@n1.Date))
   > "SetText (Text)"(self=Get BuildNumber() -> (BuildNumber:Text Object Reference), InText="To Text (String)"(InString=@n1.build))
   > "SetText (Text)"(self=Get objectlist() -> (objectlist:Text Object Reference), InText="To Text (String)"(InString=@n1.objectList))
   > "SetText (Text)"(self=Get Pass(), InText="To Text (String)"(InString=@n1.Pass))
   > "SetText (Text)"(self=Get frameNumber(), InText="To Text (String)"(InString=@n1.FrameNumber))
   > "SetText (Text)"(self=Get Version() -> (Version:Text Object Reference), InText="To Text (String)"(InString=@n1.Version))
  ]}
]
variables []
widgets {
 info {
  inheritedWidgetCount=0
  parentClass=/Script/MovieRenderPipelineSettings.MoviePipelineBurnInWidget
  rootWidgetClass=&CanvasPanel
  widgetCount=8
 }
 widgets [
  CanvasPanel_0 {widgetClassPath=&CanvasPanel, widget=CanvasPanel_0}
  Name {
   widgetClassPath=&TextBlock
   bIsVariable=true
   parent=CanvasPanel_0
   widget=Name
   properties {
    text=Name
    font {
     bForceMonospaced=false
     bMaterialIsStencil=false
     fontMaterial=null
     fontObject=&Roboto
     letterSpacing=0
     monospacedWidth=1
     size=30
     skewAmount=0
     typefaceFontName=Bold
     outlineSettings {
      bApplyOutlineToDropShadows=false
      bMiteredCorners=false
      bSeparateFillAlpha=false
      outlineMaterial=null
      outlineSize=0
      outlineColor=<rgba 0 0 0 1>
     }
    }
   }
   slot {class=CanvasPanelSlot, properties {layoutData {alignment=<0 0>, anchors {maximum=<0 0>, minimum=<0 0>}, offsets=<ltrb 20 20 800 50>}}}
  }
  objectlist {
   widgetClassPath=&TextBlock
   bIsVariable=true
   parent=CanvasPanel_0
   widget=objectlist
   properties {
    justification=Right
    text=Objects
    font {
     bForceMonospaced=false
     bMaterialIsStencil=false
     fontMaterial=null
     fontObject=&Roboto
     letterSpacing=0
     monospacedWidth=1
     size=20
     skewAmount=0
     typefaceFontName=Bold
     outlineSettings {
      bApplyOutlineToDropShadows=false
      bMiteredCorners=false
      bSeparateFillAlpha=false
      outlineMaterial=null
      outlineSize=0
      outlineColor=<rgba 0 0 0 1>
     }
    }
   }
   slot {
    class=CanvasPanelSlot
    properties {layoutData {alignment=<0 0>, anchors {maximum=<1 0>, minimum=<1 0>}, offsets=<ltrb -820 168 800 582.765>}}
   }
  }
  Pass {
   widgetClassPath=&TextBlock
   bIsVariable=true
   parent=CanvasPanel_0
   widget=Pass
   properties {
    text=PassName
    font {
     bForceMonospaced=false
     bMaterialIsStencil=false
     fontMaterial=null
     fontObject=&Roboto
     letterSpacing=0
     monospacedWidth=1
     size=30
     skewAmount=0
     typefaceFontName=Bold
     outlineSettings {
      bApplyOutlineToDropShadows=false
      bMiteredCorners=false
      bSeparateFillAlpha=false
      outlineMaterial=null
      outlineSize=0
      outlineColor=<rgba 0 0 0 1>
     }
    }
   }
   slot {class=CanvasPanelSlot, properties {layoutData {alignment=<0 0>, anchors {maximum=<0 1>, minimum=<0 1>}, offsets=<ltrb 20 -70 500 50>}}}
  }
  frameNumber {
   widgetClassPath=&TextBlock
   bIsVariable=true
   parent=CanvasPanel_0
   widget=frameNumber
   properties {
    justification=Right
    text=9999
    font {
     bForceMonospaced=false
     bMaterialIsStencil=false
     fontMaterial=null
     fontObject=&Roboto
     letterSpacing=0
     monospacedWidth=1
     size=30
     skewAmount=0
     typefaceFontName=Bold
     outlineSettings {
      bApplyOutlineToDropShadows=false
      bMiteredCorners=false
      bSeparateFillAlpha=false
      outlineMaterial=null
      outlineSize=0
      outlineColor=<rgba 0 0 0 1>
     }
    }
   }
   slot {class=CanvasPanelSlot, properties {layoutData {alignment=<0 0>, anchors {maximum=<1 1>, minimum=<1 1>}, offsets=<ltrb -140 -70 120 50>}}}
  }
  BuildNumber {
   widgetClassPath=&TextBlock
   bIsVariable=true
   parent=CanvasPanel_0
   widget=BuildNumber
   properties {
    justification=Right
    text=Build
    font {
     bForceMonospaced=false
     bMaterialIsStencil=false
     fontMaterial=null
     fontObject=&Roboto
     letterSpacing=0
     monospacedWidth=1
     size=20
     skewAmount=0
     typefaceFontName=Bold
     outlineSettings {
      bApplyOutlineToDropShadows=false
      bMiteredCorners=false
      bSeparateFillAlpha=false
      outlineMaterial=null
      outlineSize=0
      outlineColor=<rgba 0 0 0 1>
     }
    }
   }
   slot {class=CanvasPanelSlot, properties {layoutData {alignment=<0 0>, anchors {maximum=<1 0>, minimum=<1 0>}, offsets=<ltrb -520 104 500 60.02>}}}
  }
  Date {
   widgetClassPath=&TextBlock
   bIsVariable=true
   parent=CanvasPanel_0
   widget=Date
   properties {
    justification=Right
    text=Date
    font {
     bForceMonospaced=false
     bMaterialIsStencil=false
     fontMaterial=null
     fontObject=&Roboto
     letterSpacing=0
     monospacedWidth=1
     size=30
     skewAmount=0
     typefaceFontName=Bold
     outlineSettings {
      bApplyOutlineToDropShadows=false
      bMiteredCorners=false
      bSeparateFillAlpha=false
      outlineMaterial=null
      outlineSize=0
      outlineColor=<rgba 0 0 0 1>
     }
    }
   }
   slot {class=CanvasPanelSlot, properties {layoutData {alignment=<0 0>, anchors {maximum=<1 0>, minimum=<1 0>}, offsets=<ltrb -520 20 500 50>}}}
  }
  Version {
   widgetClassPath=&TextBlock
   bIsVariable=true
   parent=CanvasPanel_0
   widget=Version
   properties {
    justification=Right
    text=Fortnite v4.0
    font {
     bForceMonospaced=false
     bMaterialIsStencil=false
     fontMaterial=null
     fontObject=&Roboto
     letterSpacing=0
     monospacedWidth=1
     size=20
     skewAmount=0
     typefaceFontName=Bold
     outlineSettings {
      bApplyOutlineToDropShadows=false
      bMiteredCorners=false
      bSeparateFillAlpha=false
      outlineMaterial=null
      outlineSize=0
      outlineColor=<rgba 0 0 0 1>
     }
    }
   }
   slot {class=CanvasPanelSlot, properties {layoutData {alignment=<0 0>, anchors {maximum=<1 1>, minimum=<1 1>}, offsets=<ltrb -644 -58 500 32.84>}}}
  }
 ]
}
